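# Physics asset for the jacket on the OxideHoard Jiffy medium male player body, setting up collision and joints for seven dynamic jacket panels.
# Each box or capsule Dyn panel hangs off a tiny mstr sphere anchor through a swing-limited, zero-twist joint.
# Front and side anchors link to thigh capsules, back anchors to the pelvis capsule, with free rotation.
# Free-rotation links also join neighboring back panels and left side-to-front, with no right-side equivalent.
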PhysicsAsset /BRCosmetics/Characters/Player/Male/Medium/Bodies/M_MED_OxideHoard_Jiffy/Meshes/M_MED_OxideHoard_Jiffy_Physics_Jacket
bodies {
 Dyn_jacketBack [{
   shapeType=Box
   extentX=2.8216
   extentY=13.3931
   extentZ=13.1943
   shapeName=Dyn_jacketFront_03_L_box
   center=<12.6504 -0.4665 -0.0213>
   rotation=<rot -0.52 89.9997 -0.0004>
  }]
 Dyn_jacketFront_03_L [{
   shapeType=Capsule
   length=8.7894
   radius=5.3672
   shapeName=Dyn_jacketFront_03_L_capsule
   center=<-10.3536 3.425 -2.2794>
   rotation=<rot -0.0555 -0.0554 6.8105>
  }]
 Dyn_jacketFront_03_R [{
   shapeType=Capsule
   length=8.7894
   radius=5.3672
   shapeName=Dyn_jacketFront_03_R_capsule
   center=<10.3542 -2.7151 2.1946>
   rotation=<rot -0.0555 -0.0554 -173.1895>
  }]
 Dyn_jacketSide_02_L [{
   shapeType=Box
   extentX=2.6843
   extentY=13.3931
   extentZ=9.174
   shapeName=Dyn_jacketFront_03_L_box
   center=<-7.8614 -0.398 -0.0218>
   rotation=<rot -0.52 89.9997 -0.0004>
  }]
 Dyn_jacketSide_02_R [{
   shapeType=Box
   extentX=2.6843
   extentY=13.3931
   extentZ=9.174
   shapeName=Dyn_jacketFront_03_R_box
   center=<7.8614 0.398 0.0218>
   rotation=<rot -0.52 89.9997 179.9996>
  }]
 Dyn_jacketsideBack_L [{
   shapeType=Box
   extentX=2.6843
   extentY=13.3931
   extentZ=9.174
   shapeName=Dyn_jacketFront_03_L_box
   center=<-13.0312 0 -0.8252>
   rotation=<rot 0 90 3.155>
  }]
 Dyn_jacketsideBack_R [{
   shapeType=Box
   extentX=2.6843
   extentY=13.3931
   extentZ=9.174
   shapeName=Dyn_jacketFront_03_R_box
   center=<13.0312 0 0.8252>
   rotation=<rot 0 90 -176.845>
  }]
 mstr_jacketBack [{shapeType=Sphere, radius=0.505, shapeName=mstr_jacketFront_03_R_sphere}]
 mstr_jacketFront_03_L [{shapeType=Sphere, radius=0.505, shapeName=mstr_jacketFront_03_R_sphere}]
 mstr_jacketFront_03_R [{shapeType=Sphere, radius=0.505, shapeName=mstr_jacketFront_03_R_sphere}]
 mstr_jacketSide_02_L [{shapeType=Sphere, radius=0.505, shapeName=mstr_jacketFront_03_R_sphere}]
 mstr_jacketSide_02_R [{shapeType=Sphere, radius=0.505, shapeName=mstr_jacketFront_03_R_sphere}]
 mstr_jacketsideBack_L [{shapeType=Sphere, radius=0.505, shapeName=mstr_jacketFront_03_R_sphere}]
 mstr_jacketsideBack_R [{shapeType=Sphere, radius=0.505, shapeName=mstr_jacketFront_03_R_sphere}]
 pelvis [{shapeType=Capsule, length=8.3527, radius=12.2854, shapeName=pelvis_capsule, center=<-4.9049 -0.1367 0>}]
 thigh_l [{
   shapeType=Capsule
   length=26.2052
   radius=9.1583
   shapeName=thigh_l_capsule
   center=<-17.2735 1.0969 0.8566>
   rotation=<rot 89.3936 -27.6083 -29.3007>
  }]
 thigh_r [{
   shapeType=Capsule
   length=26.2052
   radius=9.1583
   shapeName=thigh_r_capsule
   center=<17.2735 -1.0969 -0.8566>
   rotation=<rot 89.3936 -27.6083 150.6993>
  }]
}
constraints [
 {bone1Name=Dyn_jacketFront_03_L, bone2Name=mstr_jacketFront_03_L, swing1LimitDegrees=25, swing2LimitDegrees=25, twistLimitDegrees=0}
 {bone1Name=Dyn_jacketSide_02_L, bone2Name=mstr_jacketSide_02_L, swing1LimitDegrees=25, swing2LimitDegrees=25, twistLimitDegrees=0}
 {bone1Name=Dyn_jacketsideBack_L, bone2Name=mstr_jacketsideBack_L, swing1LimitDegrees=25, swing2LimitDegrees=25, twistLimitDegrees=0}
 {bone1Name=Dyn_jacketBack, bone2Name=mstr_jacketBack, swing1LimitDegrees=25, swing2LimitDegrees=25, twistLimitDegrees=0}
 {
  bone1Name=mstr_jacketFront_03_L
  bone2Name=thigh_l
  swing1LimitDegrees=45
  swing1Motion=Free
  swing2LimitDegrees=45
  swing2Motion=Free
  twistLimitDegrees=45
  twistMotion=Free
 }
 {
  bone1Name=mstr_jacketSide_02_L
  bone2Name=thigh_l
  swing1LimitDegrees=45
  swing1Motion=Free
  swing2LimitDegrees=45
  swing2Motion=Free
  twistLimitDegrees=45
  twistMotion=Free
 }
 {
  bone1Name=thigh_l
  bone2Name=pelvis
  swing1LimitDegrees=45
  swing1Motion=Free
  swing2LimitDegrees=45
  swing2Motion=Free
  twistLimitDegrees=45
  twistMotion=Free
 }
 {
  bone1Name=mstr_jacketFront_03_R
  bone2Name=thigh_r
  swing1LimitDegrees=45
  swing1Motion=Free
  swing2LimitDegrees=45
  swing2Motion=Free
  twistLimitDegrees=45
  twistMotion=Free
 }
 {
  bone1Name=mstr_jacketSide_02_R
  bone2Name=thigh_r
  swing1LimitDegrees=45
  swing1Motion=Free
  swing2LimitDegrees=45
  swing2Motion=Free
  twistLimitDegrees=45
  twistMotion=Free
 }
 {
  bone1Name=mstr_jacketsideBack_R
  bone2Name=pelvis
  swing1LimitDegrees=45
  swing1Motion=Free
  swing2LimitDegrees=45
  swing2Motion=Free
  twistLimitDegrees=45
  twistMotion=Free
 }
 {
  bone1Name=mstr_jacketsideBack_L
  bone2Name=pelvis
  swing1LimitDegrees=45
  swing1Motion=Free
  swing2LimitDegrees=45
  swing2Motion=Free
  twistLimitDegrees=45
  twistMotion=Free
 }
 {
  bone1Name=mstr_jacketBack
  bone2Name=pelvis
  swing1LimitDegrees=45
  swing1Motion=Free
  swing2LimitDegrees=45
  swing2Motion=Free
  twistLimitDegrees=45
  twistMotion=Free
 }
 {bone1Name=Dyn_jacketsideBack_R, bone2Name=mstr_jacketsideBack_R, swing1LimitDegrees=25, swing2LimitDegrees=25, twistLimitDegrees=0}
 {bone1Name=Dyn_jacketSide_02_R, bone2Name=mstr_jacketSide_02_R, swing1LimitDegrees=25, swing2LimitDegrees=25, twistLimitDegrees=0}
 {bone1Name=Dyn_jacketFront_03_R, bone2Name=mstr_jacketFront_03_R, swing1LimitDegrees=25, swing2LimitDegrees=25, twistLimitDegrees=0}
 {
  bone1Name=thigh_r
  bone2Name=pelvis
  swing1LimitDegrees=45
  swing1Motion=Free
  swing2LimitDegrees=45
  swing2Motion=Free
  twistLimitDegrees=45
  twistMotion=Free
 }
 {
  bone1Name=Dyn_jacketBack
  bone2Name=Dyn_jacketsideBack_R
  swing1LimitDegrees=45
  swing1Motion=Free
  swing2LimitDegrees=45
  swing2Motion=Free
  twistLimitDegrees=45
  twistMotion=Free
 }
 {
  bone1Name=Dyn_jacketBack
  bone2Name=Dyn_jacketsideBack_L
  swing1LimitDegrees=45
  swing1Motion=Free
  swing2LimitDegrees=45
  swing2Motion=Free
  twistLimitDegrees=45
  twistMotion=Free
 }
 {
  bone1Name=Dyn_jacketSide_02_L
  bone2Name=Dyn_jacketFront_03_L
  swing1LimitDegrees=45
  swing1Motion=Free
  swing2LimitDegrees=45
  swing2Motion=Free
  twistLimitDegrees=45
  twistMotion=Free
 }
]
