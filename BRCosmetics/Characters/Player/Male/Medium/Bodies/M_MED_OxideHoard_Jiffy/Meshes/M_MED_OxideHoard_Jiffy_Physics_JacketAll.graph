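# Physics asset for the dynamic jacket on the medium male OxideHoard Jiffy player body cosmetic.
# Box and capsule bodies on Dyn_ bones form front, side and back panels, chained to small mstr_ anchor spheres with tight swing and twist limits.
# Free-angle constraints link lower side panels to neighboring front and side-back panels, and front and side anchors to spine_02.
# Pelvis, spine and thigh capsules and spheres provide body volumes around the jacket.
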PhysicsAsset /BRCosmetics/Characters/Player/Male/Medium/Bodies/M_MED_OxideHoard_Jiffy/Meshes/M_MED_OxideHoard_Jiffy_Physics_JacketAll
bodies {
 Dyn_jacketBack [{
   shapeType=Box
   extentX=2.8216
   extentY=13.3931
   extentZ=13.1943
   shapeName=Dyn_jacketFront_03_L_box
   center=<12.6504 -0.4665 -0.0213>
   rotation=<rot -0.52 89.9997 -0.0004>
  }]
 Dyn_jacketFront_01_L [{
   shapeType=Box
   extentX=2.6843
   extentY=13.3931
   extentZ=9.174
   shapeName=Dyn_jacketFront_03_L_box
   center=<-7.8614 -0.398 -0.0218>
   rotation=<rot -0.52 89.9997 -0.0004>
  }]
 Dyn_jacketFront_01_R [{
   shapeType=Box
   extentX=2.6843
   extentY=13.3931
   extentZ=9.174
   shapeName=Dyn_jacketFront_03_R_box
   center=<7.4025 0.5276 0.9825>
   rotation=<rot -3.6403 89.9998 179.9999>
  }]
 Dyn_jacketFront_02_L [{
   shapeType=Box
   extentX=2.6843
   extentY=13.3931
   extentZ=9.174
   shapeName=Dyn_jacketFront_03_L_box
   center=<-7.8614 -0.398 -0.0218>
   rotation=<rot -0.52 89.9997 -0.0004>
  }]
 Dyn_jacketFront_02_R [{
   shapeType=Box
   extentX=2.6843
   extentY=13.3931
   extentZ=9.174
   shapeName=Dyn_jacketFront_03_R_box
   center=<7.4025 0.5276 0.9825>
   rotation=<rot -3.6403 89.9998 179.9999>
  }]
 Dyn_jacketFront_03_L [{
   shapeType=Capsule
   length=8.7894
   radius=5.3672
   shapeName=Dyn_jacketFront_03_L_capsule
   center=<-10.3542 2.7151 -2.1946>
   rotation=<rot -0.0555 -0.0554 6.8105>
  }]
 Dyn_jacketFront_03_R [{
   shapeType=Capsule
   length=8.7894
   radius=5.3672
   shapeName=Dyn_jacketFront_03_R_capsule
   center=<10.3542 -2.7151 2.1946>
   rotation=<rot -0.0555 -0.0554 -173.1895>
  }]
 Dyn_jacketSide_01_L [{
   shapeType=Box
   extentX=2.6843
   extentY=9.1539
   extentZ=9.174
   shapeName=Dyn_jacketFront_03_L_box
   center=<-7.8614 0.5161 -0.0301>
   rotation=<rot -0.52 89.9997 -0.0004>
  }]
 Dyn_jacketSide_01_R [{
   shapeType=Box
   extentX=2.6843
   extentY=13.3931
   extentZ=9.174
   shapeName=Dyn_jacketFront_03_R_box
   center=<7.8614 0.398 0.0218>
   rotation=<rot -0.52 89.9997 179.9996>
  }]
 Dyn_jacketSide_02_L [{
   shapeType=Box
   extentX=2.6843
   extentY=13.3931
   extentZ=9.174
   shapeName=Dyn_jacketFront_03_L_box
   center=<-7.8614 0.3479 -0.0285>
   rotation=<rot -0.5188 93.8074 -0.0351>
  }]
 Dyn_jacketSide_02_R [{
   shapeType=Box
   extentX=2.6843
   extentY=13.3931
   extentZ=9.174
   shapeName=Dyn_jacketFront_03_R_box
   center=<7.8614 0.398 0.0218>
   rotation=<rot -0.52 89.9997 179.9996>
  }]
 Dyn_jacketsideBack_L [{
   shapeType=Box
   extentX=2.6843
   extentY=13.3931
   extentZ=9.174
   shapeName=Dyn_jacketFront_03_L_box
   center=<-13.0312 0 -0.8252>
   rotation=<rot 0 90 3.155>
  }]
 Dyn_jacketsideBack_R [{
   shapeType=Box
   extentX=2.6843
   extentY=13.3931
   extentZ=9.174
   shapeName=Dyn_jacketFront_03_R_box
   center=<13.0312 0 0.8252>
   rotation=<rot 0 90 -176.845>
  }]
 mstr_jacketBack [{shapeType=Sphere, radius=0.505, shapeName=mstr_jacketFront_03_R_sphere}]
 mstr_jacketFront_01_L [{shapeType=Sphere, radius=0.505, shapeName=mstr_jacketFront_01_L_sphere}]
 mstr_jacketFront_01_R [{shapeType=Sphere, radius=0.505, shapeName=mstr_jacketFront_01_R_sphere}]
 mstr_jacketFront_03_L [{shapeType=Sphere, radius=0.505, shapeName=mstr_jacketFront_03_R_sphere}]
 mstr_jacketSide_01_L [{shapeType=Sphere, radius=0.505, shapeName=mstr_jacketSide_01_L_sphere}]
 mstr_jacketSide_01_R [{shapeType=Sphere, radius=0.505, shapeName=mstr_jacketSide_01_R_sphere}]
 mstr_jacketSide_02_L [{shapeType=Sphere, radius=0.505, shapeName=mstr_jacketFront_03_R_sphere}]
 mstr_jacketsideBack_L [{shapeType=Sphere, radius=0.505, shapeName=mstr_jacketFront_03_R_sphere}]
 mstr_jacketsideBack_R [{shapeType=Sphere, radius=0.505, shapeName=mstr_jacketFront_03_R_sphere}]
 pelvis [{shapeType=Capsule, length=7.8014, radius=12.3951, shapeName=pelvis_capsule, center=<-4.9049 -0.2271 0>}]
 spine_01 [{shapeType=Capsule, length=5.2655, radius=14.6627, shapeName=pelvis_capsule, center=<-4.905 -0.591 0>, rotation=<rot 90 -0.0029 0>}]
 spine_02 [{shapeType=Capsule, length=7.0476, radius=15.3378, shapeName=pelvis_capsule, center=<-0.1574 -1.1677 0>, rotation=<rot 90 3.076 0>}]
 thigh_l [
  {shapeType=Sphere, radius=10.7504, shapeName=thigh_l_sphere, center=<-8.4272 -1.3422 2.3317>}
  {
   shapeType=Capsule
   length=14.4373
   radius=11.0635
   shapeName=thigh_l_capsule
   center=<-22.7986 0.9335 0.9076>
   rotation=<rot 89.3936 -27.6083 -29.3007>
  }
 ]
 thigh_r [
  {shapeType=Sphere, radius=10.7504, shapeName=thigh_r_sphere, center=<8.4272 1.3422 -2.3317>}
  {
   shapeType=Capsule
   length=14.4373
   radius=11.0635
   shapeName=thigh_r_capsule
   center=<22.7986 -0.9335 -0.9076>
   rotation=<rot 89.3936 -27.6083 150.6993>
  }
 ]
}
constraints [
 {bone1Name=Dyn_jacketsideBack_L, bone2Name=mstr_jacketsideBack_L, swing1LimitDegrees=25, swing2LimitDegrees=25, twistLimitDegrees=0}
 {bone1Name=Dyn_jacketBack, bone2Name=mstr_jacketBack, swing1LimitDegrees=25, swing2LimitDegrees=25, twistLimitDegrees=0}
 {bone1Name=Dyn_jacketsideBack_R, bone2Name=mstr_jacketsideBack_R, swing1LimitDegrees=25, swing2LimitDegrees=25, twistLimitDegrees=0}
 {bone1Name=Dyn_jacketBack, bone2Name=Dyn_jacketsideBack_R, swing1LimitDegrees=25, swing2LimitDegrees=25, twistLimitDegrees=0}
 {bone1Name=Dyn_jacketBack, bone2Name=Dyn_jacketsideBack_L, swing1LimitDegrees=25, swing2LimitDegrees=25, twistLimitDegrees=0}
 {bone1Name=Dyn_jacketFront_03_L, bone2Name=Dyn_jacketFront_02_L, swing1LimitDegrees=25, swing2LimitDegrees=10, twistLimitDegrees=10}
 {bone1Name=Dyn_jacketFront_02_L, bone2Name=Dyn_jacketFront_01_L, swing1LimitDegrees=25, swing2LimitDegrees=10, twistLimitDegrees=10}
 {bone1Name=Dyn_jacketFront_01_L, bone2Name=mstr_jacketFront_01_L, swing1LimitDegrees=25, swing2LimitDegrees=25, twistLimitDegrees=0}
 {bone1Name=Dyn_jacketSide_02_L, bone2Name=Dyn_jacketSide_01_L, swing1LimitDegrees=25, swing2LimitDegrees=10, twistLimitDegrees=10}
 {bone1Name=Dyn_jacketSide_01_L, bone2Name=mstr_jacketSide_01_L, swing1LimitDegrees=25, swing2LimitDegrees=25, twistLimitDegrees=0}
 {bone1Name=Dyn_jacketSide_02_R, bone2Name=Dyn_jacketSide_01_R, swing1LimitDegrees=25, swing2LimitDegrees=25, twistLimitDegrees=0}
 {bone1Name=Dyn_jacketSide_01_R, bone2Name=mstr_jacketSide_01_R, swing1LimitDegrees=25, swing2LimitDegrees=25, twistLimitDegrees=0}
 {bone1Name=Dyn_jacketFront_03_R, bone2Name=Dyn_jacketFront_02_R, swing1LimitDegrees=25, swing2LimitDegrees=25, twistLimitDegrees=0}
 {bone1Name=Dyn_jacketFront_02_R, bone2Name=Dyn_jacketFront_01_R, swing1LimitDegrees=25, swing2LimitDegrees=10, twistLimitDegrees=10}
 {bone1Name=Dyn_jacketFront_01_R, bone2Name=mstr_jacketFront_01_R, swing1LimitDegrees=25, swing2LimitDegrees=25, twistLimitDegrees=0}
 {
  bone1Name=mstr_jacketFront_01_R
  bone2Name=spine_02
  swing1LimitDegrees=45
  swing1Motion=Free
  swing2LimitDegrees=45
  swing2Motion=Free
  twistLimitDegrees=45
  twistMotion=Free
 }
 {
  bone1Name=mstr_jacketFront_01_L
  bone2Name=spine_02
  swing1LimitDegrees=45
  swing1Motion=Free
  swing2LimitDegrees=45
  swing2Motion=Free
  twistLimitDegrees=45
  twistMotion=Free
 }
 {
  bone1Name=mstr_jacketSide_01_R
  bone2Name=spine_02
  swing1LimitDegrees=45
  swing1Motion=Free
  swing2LimitDegrees=45
  swing2Motion=Free
  twistLimitDegrees=45
  twistMotion=Free
 }
 {
  bone1Name=mstr_jacketSide_01_L
  bone2Name=spine_02
  swing1LimitDegrees=45
  swing1Motion=Free
  swing2LimitDegrees=45
  swing2Motion=Free
  twistLimitDegrees=45
  twistMotion=Free
 }
 {
  bone1Name=spine_01
  bone2Name=pelvis
  swing1LimitDegrees=45
  swing1Motion=Free
  swing2LimitDegrees=45
  swing2Motion=Free
  twistLimitDegrees=45
  twistMotion=Free
 }
 {
  bone1Name=spine_02
  bone2Name=spine_01
  swing1LimitDegrees=45
  swing1Motion=Free
  swing2LimitDegrees=45
  swing2Motion=Free
  twistLimitDegrees=45
  twistMotion=Free
 }
 {
  bone1Name=Dyn_jacketsideBack_L
  bone2Name=Dyn_jacketSide_02_L
  swing1LimitDegrees=45
  swing1Motion=Free
  swing2LimitDegrees=45
  swing2Motion=Free
  twistLimitDegrees=45
  twistMotion=Free
 }
 {
  bone1Name=Dyn_jacketsideBack_R
  bone2Name=Dyn_jacketSide_02_R
  swing1LimitDegrees=45
  swing1Motion=Free
  swing2LimitDegrees=45
  swing2Motion=Free
  twistLimitDegrees=45
  twistMotion=Free
 }
 {
  bone1Name=Dyn_jacketSide_02_L
  bone2Name=Dyn_jacketFront_03_L
  swing1LimitDegrees=45
  swing1Motion=Free
  swing2LimitDegrees=45
  swing2Motion=Free
  twistLimitDegrees=45
  twistMotion=Free
 }
 {
  bone1Name=Dyn_jacketSide_02_R
  bone2Name=Dyn_jacketFront_03_R
  swing1LimitDegrees=45
  swing1Motion=Free
  swing2LimitDegrees=45
  swing2Motion=Free
  twistLimitDegrees=45
  twistMotion=Free
 }
]
