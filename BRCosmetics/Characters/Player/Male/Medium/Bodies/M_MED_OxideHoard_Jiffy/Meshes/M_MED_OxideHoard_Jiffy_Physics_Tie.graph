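# Physics asset for the tie and front jacket flaps on the male medium OxideHoard Jiffy character body cosmetic.
# Capsules on the pelvis, spine and thighs form the body volume, while boxes and small capsules cover the dynamic tie and jacket bones.
# The five tie bodies chain down from spine_05 with tight angular limits that loosen slightly after the second link.
# The jacket bodies are not chained: each pins separately to spine_03, spine_02 or its side's thigh with free swing and twist.
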PhysicsAsset /BRCosmetics/Characters/Player/Male/Medium/Bodies/M_MED_OxideHoard_Jiffy/Meshes/M_MED_OxideHoard_Jiffy_Physics_Tie
bodies {
 Dyn_jacketFront_01_L [{
   shapeType=Box
   extentX=2.6843
   extentY=13.3931
   extentZ=9.174
   shapeName=Dyn_jacketFront_03_L_box
   center=<-7.8614 -0.398 -0.0218>
   rotation=<rot -0.52 89.9997 -0.0004>
  }]
 Dyn_jacketFront_01_R [{
   shapeType=Box
   extentX=2.6843
   extentY=13.3931
   extentZ=9.174
   shapeName=Dyn_jacketFront_03_R_box
   center=<7.8614 0.398 0.0218>
   rotation=<rot -0.52 89.9997 179.9996>
  }]
 Dyn_jacketFront_02_L [{
   shapeType=Box
   extentX=2.6843
   extentY=13.3931
   extentZ=9.174
   shapeName=Dyn_jacketFront_03_L_box
   center=<-7.8614 -0.398 -0.0218>
   rotation=<rot -0.52 89.9997 -0.0004>
  }]
 Dyn_jacketFront_02_R [{
   shapeType=Box
   extentX=2.6843
   extentY=13.3931
   extentZ=9.174
   shapeName=Dyn_jacketFront_03_R_box
   center=<7.8614 0.398 0.0218>
   rotation=<rot -0.52 89.9997 179.9996>
  }]
 Dyn_jacketFront_03_L [{
   shapeType=Box
   extentX=2.6843
   extentY=13.3931
   extentZ=9.174
   shapeName=Dyn_jacketFront_03_L_box
   center=<-7.4025 -0.5276 -0.9825>
   rotation=<rot -3.6403 89.9998 -0.0001>
  }]
 Dyn_jacketFront_03_R [{
   shapeType=Box
   extentX=2.6843
   extentY=13.3931
   extentZ=9.174
   shapeName=Dyn_jacketFront_03_R_box
   center=<7.4025 0.5276 0.9825>
   rotation=<rot -3.6403 89.9998 179.9999>
  }]
 dyn_tie_01 [{
   shapeType=Box
   extentX=7.6243
   extentY=2.4629
   extentZ=6.345
   shapeName=dyn_tie_05_box
   center=<8.6936 0 0.3452>
   rotation=<rot -6.3658 0 90>
  }]
 dyn_tie_02 [{
   shapeType=Box
   extentX=8.9692
   extentY=2.773
   extentZ=9.4895
   shapeName=dyn_tie_02_box
   center=<10.1825 0 0.5423>
   rotation=<rot 0 90 86.5176>
  }]
 dyn_tie_03 [{
   shapeType=Box
   extentX=9.3129
   extentY=2.6665
   extentZ=9.2976
   shapeName=dyn_tie_03_box
   center=<6.0822 -0.2503 0.5235>
   rotation=<rot 0 -88.6747 -87.725>
  }]
 dyn_tie_04 [{shapeType=Capsule, length=3.8531, radius=3.3196, shapeName=dyn_tie_04_capsule, center=<2.7668 0 1.3699>, rotation=<rot 0 0 90>}]
 dyn_tie_05 [{shapeType=Capsule, length=3.8531, radius=3.3196, shapeName=dyn_tie_04_capsule, center=<3.244 0 1.3699>, rotation=<rot 0 0 90>}]
 pelvis [{shapeType=Capsule, length=6.6167, radius=14.6832, shapeName=pelvis_capsule, center=<-0.6994 0 0>}]
 spine_01 [{shapeType=Capsule, length=7.8496, radius=15.6775, shapeName=pelvis_capsule, center=<0 -1.4483 0>, rotation=<rot 90 -0.0001 0>}]
 spine_02 [{shapeType=Capsule, length=4.9226, radius=14.4166, shapeName=pelvis_capsule, center=<1.5028 0 0>}]
 spine_03 [{shapeType=Capsule, length=7.4661, radius=14.293, shapeName=pelvis_capsule}]
 spine_04 [{shapeType=Capsule, length=12.1765, radius=13.3231, shapeName=pelvis_capsule}]
 spine_05 [{shapeType=Capsule, length=15.499, radius=11.6611, shapeName=pelvis_capsule, center=<-3.6459 0 0>}]
 thigh_l [{
   shapeType=Capsule
   length=26.0592
   radius=9.5766
   shapeName=thigh_l_capsule
   center=<-12.9205 2.533 0>
   rotation=<rot 88.3953 2.0356 -179.9987>
  }]
 thigh_r [{shapeType=Capsule, length=26.0592, radius=9.5766, shapeName=thigh_r_capsule, center=<12.9205 -2.533 0>, rotation=<rot 87.6289 2.0225 0.0004>}]
}
constraints [
 {bone1Name=dyn_tie_01, bone2Name=spine_05, swing1LimitDegrees=10, swing2LimitDegrees=10, twistLimitDegrees=10}
 {bone1Name=dyn_tie_02, bone2Name=dyn_tie_01, swing1LimitDegrees=10, swing2LimitDegrees=10, twistLimitDegrees=10}
 {bone1Name=dyn_tie_03, bone2Name=dyn_tie_02, swing1LimitDegrees=15, swing2LimitDegrees=15, twistLimitDegrees=15}
 {bone1Name=dyn_tie_04, bone2Name=dyn_tie_03, swing1LimitDegrees=15, swing2LimitDegrees=15, twistLimitDegrees=15}
 {bone1Name=dyn_tie_05, bone2Name=dyn_tie_04, swing1LimitDegrees=15, swing2LimitDegrees=15, twistLimitDegrees=15}
 {
  bone1Name=Dyn_jacketFront_02_L
  bone2Name=spine_02
  swing1LimitDegrees=45
  swing1Motion=Free
  swing2LimitDegrees=45
  swing2Motion=Free
  twistLimitDegrees=45
  twistMotion=Free
 }
 {
  bone1Name=Dyn_jacketFront_02_R
  bone2Name=spine_02
  swing1LimitDegrees=45
  swing1Motion=Free
  swing2LimitDegrees=45
  swing2Motion=Free
  twistLimitDegrees=45
  twistMotion=Free
 }
 {
  bone1Name=Dyn_jacketFront_01_L
  bone2Name=spine_03
  swing1LimitDegrees=45
  swing1Motion=Free
  swing2LimitDegrees=45
  swing2Motion=Free
  twistLimitDegrees=45
  twistMotion=Free
 }
 {
  bone1Name=Dyn_jacketFront_01_R
  bone2Name=spine_03
  swing1LimitDegrees=45
  swing1Motion=Free
  swing2LimitDegrees=45
  swing2Motion=Free
  twistLimitDegrees=45
  twistMotion=Free
 }
 {
  bone1Name=Dyn_jacketFront_03_L
  bone2Name=thigh_l
  swing1LimitDegrees=45
  swing1Motion=Free
  swing2LimitDegrees=45
  swing2Motion=Free
  twistLimitDegrees=45
  twistMotion=Free
 }
 {
  bone1Name=Dyn_jacketFront_03_R
  bone2Name=thigh_r
  swing1LimitDegrees=45
  swing1Motion=Free
  swing2LimitDegrees=45
  swing2Motion=Free
  twistLimitDegrees=45
  twistMotion=Free
 }
]
