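# Animation Blueprint for the OxideHoard Jiffy body, adding jacket and tie physics to the copied parent mesh pose.
# A Control Rig and spine bone modifiers feed a cached pose into two jacket RigidBody setups, blended by bool before tie simulation.
# The event graph clamps spine Y/Z scales and sets a full-jacket flag off during emote, crouch or tactical sprint.
# Skydiving fades out jacket physics and raises material cloth wind and turbulence.
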AnimBlueprint /BRCosmetics/Characters/Player/Male/Medium/Bodies/M_MED_OxideHoard_Jiffy/Meshes/M_MED_OxideHoard_Jiffy_AnimBP
target_skeleton=&/M_MED_OxideHoard_Jiffy_Skeleton
graphs [
 AnimGraph {nodes [
   RigidBody() {
    overridePhysicsAsset=&/M_MED_OxideHoard_Jiffy_Physics_Tie
    simulationSpace=BaseBoneSpace
    baseBoneRef {boneName=spine_05}
    simSpaceSettings {maxAngularAcceleration=100, maxAngularVelocity=10, maxLinearAcceleration=50, maxLinearVelocity=10, worldAlpha=1}
   }
   n1 = Get JacketDynAlpha() -> (JacketDynAlpha:double)
   n2 = RigidBody(ComponentPose=Local To Component(LocalPose=Use cached pose 'PreSim'()), Alpha=@n1) {
    overridePhysicsAsset=&/M_MED_OxideHoard_Jiffy_Physics_Jacket
    simulationSpace=BaseBoneSpace
    baseBoneRef {boneName=pelvis}
    simSpaceSettings {maxAngularAcceleration=25, maxAngularVelocity=5, maxLinearAcceleration=20, maxLinearVelocity=5, worldAlpha=0.2}
   }
   n3 = Component To Local(ComponentPose=@n2)
   n4 = RigidBody(ComponentPose=Local To Component(LocalPose=Use cached pose 'PreSim'()), Alpha=@n1) {
    overridePhysicsAsset=&/M_MED_OxideHoard_Jiffy_Physics_JacketAll
    simulationSpace=BaseBoneSpace
    baseBoneRef {boneName=pelvis}
    simSpaceSettings {maxAngularAcceleration=25, maxAngularVelocity=5, maxLinearAcceleration=20, maxLinearVelocity=5, worldAlpha=0.2}
   }
   n5 = Component To Local(ComponentPose=@n4)
   n6 = RigidBody(ComponentPose=Local To Component(LocalPose=Blend Poses by bool(BlendPose_0=@n5, BlendPose_1=@n3, BlendTime_1=0.3, bActiveValue=true))) {
    overridePhysicsAsset=&M_MED_OxideHoard_Jiffy_Physics_Tie
    simulationSpace=BaseBoneSpace
    baseBoneRef {boneName=spine_05}
    simSpaceSettings {maxAngularAcceleration=100, maxAngularVelocity=10, maxLinearAcceleration=50, maxLinearVelocity=10, worldAlpha=1}
   }
   Output Pose(Result=Component To Local(ComponentPose=@n6)) {layerGroup=None}
   n7 = Control Rig(Source=Copy Pose From Mesh() {bUseAttachedParent=true}) {controlRigAssetReference {blueprintRigClass=&/M_MED_OxideHoard_Jiffy_CtrlRig.M_MED_OxideHoard_Jiffy_CtrlRig_C}}
   n8 = "Transform (Modify) Bone - Bone: spine_01"(ComponentPose=Local To Component(LocalPose=@n7), Alpha=1, Translation="0,0,0", Rotation="0,0,0", Scale="1,1,1") {scaleMode=BMM_Replace, boneToModify {boneName=spine_01}}
   n9 = "Transform (Modify) Bone - Bone: spine_04"(
    ComponentPose="Transform (Modify) Bone - Bone: spine_03"(
     ComponentPose="Transform (Modify) Bone - Bone: spine_02"(ComponentPose=@n8, Alpha=1, Translation="0,0,0", Rotation="0,0,0", Scale="1,1,1") {scaleMode=BMM_Replace, boneToModify {boneName=spine_02}}
     Alpha=1
     Translation="0,0,0"
     Rotation="0,0,0"
     Scale="1,1,1"
    ) {scaleMode=BMM_Replace, boneToModify {boneName=spine_03}}
    Alpha=1
    Translation="0,0,0"
    Rotation="0,0,0"
    Scale="1,1,1"
   ) {scaleMode=BMM_Replace, boneToModify {boneName=spine_04}}
   Save cached pose 'PreSim'(Pose=Component To Local(ComponentPose="Transform (Modify) Bone - Bone: spine_05"(ComponentPose=@n9, Alpha=1, Translation="0,0,0", Rotation="0,0,0", Scale="1,1,1") {scaleMode=BMM_Replace, boneToModify {boneName=spine_05}})) {cacheName=PreSim}
  ]}
 Clamp Vector YZ {nodes [
   n1 = Inputs!() -> (InVec:Vector=~, min:double=~, max:double=1)
   n2 = BreakVector(InVec=@n1)
   "Clamp (Float)"(Value=@n2, Min=@n1.min, Max=@n1.max)
   Outputs!(ReturnValue:Vector=MakeVector(X=@n2, Y="Clamp (Float)"(Value=@n2.Y, Min=@n1.min, Max=@n1.max), Z="Clamp (Float)"(Value=@n2.Z, Min=@n1.min, Max=@n1.max)))
  ]}
 EventGraph {nodes [
   Event BlueprintUpdateAnimation()
   > n1 = Sequence()
   > n2 = Sequence!(execute) -> (then_0, then_1, then_2:Exec, then_3:Exec, then_4:Exec)
   n3 = GetOwnerSkeletalMeshComponent()
   > Set Scale_spine_01(Scale_spine_01:Vector=Clamp Vector YZ!(
     InVec:Vector=GetSocketTransform!(self=@n3, InSocketName=spine_01, TransformSpace=RTS_ParentBoneSpace) -> (ReturnValue_Location:Vector, ReturnValue_Rotation:Rotator, ReturnValue_Scale:Vector).ReturnValue_Scale
     min:double
     max:double=1
    ) -> (ReturnValue:Vector=~) {is=title K2Node_MacroInstance}) -> (Output_Get:Vector)
   Set Scale_spine_02(
    execute=@n2.then_1
    Scale_spine_02:Vector=Clamp Vector YZ(InVec=GetSocketTransform(self=@n3, InSocketName=spine_02, TransformSpace=RTS_ParentBoneSpace).ReturnValue_Scale, max=1) -> (ReturnValue=~)
   ) -> (Output_Get:Vector)
   Set Scale_spine_03(
    execute=@n2.then_2
    Scale_spine_03:Vector=Clamp Vector YZ(InVec=GetSocketTransform(self=@n3, InSocketName=spine_03, TransformSpace=RTS_ParentBoneSpace).ReturnValue_Scale, max=1) -> (ReturnValue=~)
   ) -> (Output_Get:Vector)
   Set Scale_spine_04(
    execute=@n2.then_3
    Scale_spine_04:Vector=Clamp Vector YZ(InVec=GetSocketTransform(self=@n3, InSocketName=spine_04, TransformSpace=RTS_ParentBoneSpace).ReturnValue_Scale, max=1) -> (ReturnValue=~)
   ) -> (Output_Get:Vector)
   Set Scale_spine_05(
    execute=@n2.then_4
    Scale_spine_05:Vector=Clamp Vector YZ(InVec=GetSocketTransform(self=@n3, InSocketName=spine_05, TransformSpace=RTS_ParentBoneSpace).ReturnValue_Scale, max=1) -> (ReturnValue=~)
   ) -> (Output_Get:Vector)
   n4 = Get bIsPlayingEmote()
   n5 = Get bIsCrouching()
   n6 = Get bIsTacticalSprinting()
   Set FullJacketRigidBody?(execute=@n1.then_1, FullJacketRigidBody?:bool=NOT Boolean(A=OR Boolean!(A=@n4, B=@n5, C:bool=@n6) -> (ReturnValue))) -> (Output_Get:bool)
   n7 = Get bIsSkyDiving()
   > Set JacketDynAlpha(JacketDynAlpha:double=FInterpTo(
     Current=Get JacketDynAlpha()
     Target=SelectFloat(A=0, B=0.5, bPickA=@n7)
     DeltaTime=GetWorldDeltaSeconds!() -> (ReturnValue:double)
     InterpSpeed=5
    )) -> (Output_Get:double)
   n8 = Get bIsSkyDiving()
   n9 = Get Material Instance Dynamic() -> (Material Instance Dynamic:Material Instance Dynamic Object Reference)
   > SetScalarParameterValue(self=@n9, ParameterName=ClothSim_R2_TurbulenceStrength, Value=SelectFloat(A=1.5, B=0, bPickA=@n8))
   > SetScalarParameterValue(self=@n9, ParameterName=ClothSim_R1_TurbulenceStrength, Value=SelectFloat(A=1.5, B=0, bPickA=@n8))
   > SetScalarParameterValue(self=@n9, ParameterName=ClothSim_WindStrength, Value=SelectFloat(A=1, B=0, bPickA=@n8))
   n10 = Event BlueprintInitializeAnimation()
   n11 = Get Dynamic Material Instance(execute=[@n10.then, @n12.then], SkeletalMeshComponent=GetPartSkeletalMeshComponent())
   > Set Material Instance Dynamic(Material Instance Dynamic:Material Instance Dynamic Object Reference=@n11) -> (Output_Get:Material Instance Dynamic Object Reference)
   n12 = Event OnCharacterCustomizationLoaded()
  ]}
]
variables [
 FullJacketRigidBody? {type=boolean}
 JacketDynAlpha {type=number, default=0.5}
 Material Instance Dynamic {type=MaterialInstanceDynamic}
 Scale_spine_01 {type=Vector}
 Scale_spine_02 {type=Vector}
 Scale_spine_03 {type=Vector}
 Scale_spine_04 {type=Vector}
 Scale_spine_05 {type=Vector}
]
